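# Masked unlit material for hologram glitch effects in the NeonCity set, usable on Niagara sprites.
# Tiled screen-space UVs snap to a grid whose resolution steps up in a repeating time-driven cycle, producing shifting blocky samples of a glitch map's clamped green channel.
# Particle color is multiplied by a scanline function, which a path switch replaces with a constant on low-end and mobile.
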Material /Game/Environments/Asteria/Sets/NeonCity/Techart/Holograms/HologramEffects/HologramGlitch
graph {
 properties {
  bCastRayTracedShadows=false
  bUsedWithNiagaraSprites=true
  blendMode=BLEND_Masked
  pixelDepthOffsetMode=PDOM_Legacy
  shadingModel=MSM_Unlit
  usedShadingModels=Unlit
 }
 nodes [
  n1 = ScreenPosition()
  n2 = ScalarParameter() {defaultValue=9, parameterName=Tiling}
  n3 = ScalarParameter() {defaultValue=8, parameterName=Quantize}
  n4 = Time()
  n5 = ScalarParameter() {defaultValue=0.2, parameterName=GlitchSpeed}
  n6 = Constant() {r=10}
  n7 = Multiply(A=Frac(None=Multiply(A=@n4, B=@n5)), B=@n6)
  n8 = LinearInterpolate(A=Multiply(A=@n3) {constB=0.2}, B=@n3, Alpha=Divide(A=Round(None=@n7), B=@n6))
  n9 = Round(None=Multiply(A=Multiply(A=@n1.ViewportUV, B=@n2), B=@n8))
  Clamp(None=TextureSample(UVs=Divide(A=@n9, B=@n8)) {samplerType=SAMPLERTYPE_LinearColor, texture=/Game/Environments/Asteria/Sets/NeonCity/Techart/Holograms/GlitchMap_M}.G)
  n10 = Constant() {r=100}
  Multiply(
   A=ParticleColor().RGB
   B=MaterialFunctionCall!(
    Default=MaterialFunctionCall!(Scanline Amount=@n10) {materialFunction=/Game/Environments/Asteria/Sets/NeonCity/Techart/Holograms/NewHologramMaterial/MFScanlines}.Result
    Low/Mobile=Constant() {r=1}
   ) {materialFunction=/Game/Effects/Fort_Effects/Materials/_Functions/MF_LowMobile_PathingSwitch}.Result
  )
 ]
}
